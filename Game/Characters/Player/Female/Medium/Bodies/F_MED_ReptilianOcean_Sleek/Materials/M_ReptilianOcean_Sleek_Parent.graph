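# Parent character material for the female medium ReptilianOcean Sleek body, adding animated reptile-scale effects between the shared CharacterParent PreFX and PostFX functions.
# Time-driven sine sweeps along pre-skinned height, broken up by random scale masks and noise, rotate normals and drive world position offset.
# Base color layers thin-film iridescence, curve-atlas spots and animated SDF line patterns.
# Static switches gate scales, patterns, eye marks, panning emissive and snake-eye glow.
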Material /Game/Characters/Player/Female/Medium/Bodies/F_MED_ReptilianOcean_Sleek/Materials/M_ReptilianOcean_Sleek_Parent
graph {
 properties {
  bUseMaterialAttributes=true
  bUsedWithMeshDeformer=true
  bUsedWithMorphTargets=true
  bUsedWithNanite=true
  bUsedWithSkeletalMesh=true
  pixelDepthOffsetMode=PDOM_Legacy
  previewMesh=/BRCosmetics/Characters/Player/Female/Medium/Bodies/F_MED_ReptilianOcean_Battle/Meshes/F_MED_ReptilianOcean_Battle
  shadingModel=MSM_FromMaterialExpression
  usedShadingModels=Default Lit | Subsurface | Clear Coat | Subsurface Profile | Cloth
 }
 nodes [
  n1 = Add(A=NamedRerouteUsage(), B=Multiply(A=Time()) {constB=0.05})
  Clamp(None=Abs(None=Sine(None=@n1))) {minDefault=0.1}
  Constant() {r=-0.25}
  Constant() {r=1}
  n2 = OneMinus(None=ScalarParameter() {parameterName=EyeEffect, sliderMax=1})
  n3 = Multiply(A=TextureCoordinate(), B=Constant2Vector() {g=0.6, r=2})
  n4 = Add(
   A=SphereMask(A=@n3, B=Constant2Vector() {g=0.23, r=0.85}) {attenuationRadius=0.2, hardnessPercent=0}
   B=SphereMask(A=@n3, B=Constant2Vector() {g=0.23, r=1.15}) {attenuationRadius=0.2, hardnessPercent=0}
  )
  n5 = SmoothStep(Max=Divide(A=@n2) {constB=10}, Value=SmoothStep(Min=@n2, Value=Saturate(None=@n4))) {constMax=0.05}
  LinearInterpolate(
   B=TextureSample() {texture=/Game/Characters/Player/Female/Medium/Bodies/F_MED_ReptilianOcean_Sleek/Textures/T_ReptilianOcean_Sleek_Head_D_2}.RGB
   Alpha=@n5
  )
  n6 = MaterialFunctionCall() {materialFunction=&MF_CharacterParent_2023_PreFX}
  MaterialFunctionCall(Material=NamedRerouteUsage(), M=@n6.M, Pre Skinned Local Position=@n6.M, Distance Blend=@n6.M, Fresnel=@n6.M) {materialFunction=&MF_CharacterParent_2023_PostFX}
  n7 = MaterialFunctionCall(Normal=NamedRerouteUsage(), Flatness=ScalarParameter() {defaultValue=0.8, parameterName=ExtraThinFilmNormalContribution}) {materialFunction=&FlattenNormal}
  n8 = Fresnel(Normal=Transform(None=@n7.Result)) {baseReflectFraction=0, exponent=0.1}
  n9 = Power(Base=@n8, Exp=ScalarParameter() {defaultValue=4, parameterName=ExtraThinFilmMaskExponent}) {constExponent=4}
  n10 = ComponentMask(None=@n7.Result) {g=true, r=true}
  n11 = TextureSample(UVs=Add(A=TextureCoordinate(), B=Multiply(A=@n10) {constB=0.1})) {
   samplerType=SAMPLERTYPE_Grayscale
   texture=/Game/Characters/Player/Female/Medium/Bodies/F_MED_ReptilianOcean_Battle/Textures/T_ReptilianOcean_Battle_Body_ThinFilm_SDF
  }
  n12 = Subtract(A=NamedRerouteUsage()) {constB=0.5}
  n13 = ScalarParameter() {defaultValue=0.5, parameterName=ExtraThinFilmPanningNoiseIntensity}
  n14 = Power(Base=Add(A=@n11.R, B=Multiply(A=@n12, B=@n13) {constB=0.5}), Exp=ScalarParameter() {defaultValue=2, parameterName=ExtraThinFilmPower})
  n15 = ScalarParameter() {defaultValue=1.4, parameterName=ExtraThinFilmScale}
  n16 = ScalarParameter() {defaultValue=1.4, parameterName=ExtraThinFilmScaleWaistOverride}
  n17 = NamedRerouteUsage()
  n18 = Multiply(
   A=Add(
    A=StaticSwitchParameter(True=@n8, False=@n14) {defaultValue=true, parameterName=ExtraThinFilmUseFresnel}
    B=ScalarParameter() {defaultValue=-0.55, parameterName=ExtraThinFilmOffset}
   ) {constB=0.5}
   B=StaticSwitchParameter(True=LinearInterpolate(A=@n15, B=@n16, Alpha=@n17), False=@n15) {parameterName=EnableBattleWaistExtraThinFilmOverride}
  ) {constB=1.5}
  n19 = TextureSample(UVs=ConstantBiasScale(None=Sine(None=@n18))) {texture=/Game/Global/Textures/Default/Thinfilm/T_ThinFilm_Spectrum_COLOR}
  Multiply(A=SmoothStep(Value=Saturate(None=@n9)) {constMax=0.5}, B=@n19.RGB)
  NamedRerouteDeclaration(None=ChannelMaskParameter(None=NamedRerouteUsage()) {parameterName=ScalesMask}.R) {name=ScalesMask, nodeColor=<rgba 1 0 0.4219 1>}
  n20 = GetMaterialAttributes(None=NamedRerouteUsage()) {attributeGetTypes [#1, #2, #3]}
  NamedRerouteDeclaration(None=@n20.Base Color) {name=InBaseColor, nodeColor=<rgba 0 0.7187 1 1>}
  NamedRerouteDeclaration(None=ChannelMaskParameter(None=NamedRerouteUsage()) {maskChannel=Green, parameterName=RandomScalesMask, defaultValue=<rgba 0 1 0 0>}.R) {name=RandomScalesMask, nodeColor=<rgba 0 1 0.3438 1>}
  NamedRerouteDeclaration(None=@n20.Emissive Color) {name=InEmissive, nodeColor=<rgba 0 1 0.2031 1>}
  n21 = NamedRerouteUsage()
  n22 = VectorParameter() {group=06_Advanced Emissive, parameterName=PanningEmissiveColor1, sortPriority=15, defaultValue=<rgba 0.0097 0 0.3515 1>}
  n23 = VectorParameter() {group=06_Advanced Emissive, parameterName=PanningEmissiveColor2, sortPriority=15, defaultValue=<rgba 0.3325 0.0222 0.7529 1>}
  n24 = Multiply(A=Time(), B=ScalarParameter() {defaultValue=0.05, parameterName=PanningEmissiveLinesSpeed}) {constB=0.05}
  n25 = PreSkinnedPosition()
  n26 = Divide(
   A=ComponentMask(None=VertexInterpolator(VS=@n25).PS) {b=true}
   B=ScalarParameter() {defaultValue=125, parameterName=PanningEmissiveLinesAmount}
  ) {constB=-125}
  n27 = Abs(None=Sine(None=Add(A=@n24, B=@n26)))
  n28 = ScalarParameter() {defaultValue=4, parameterName=PanningEmissiveLinesSmoothness}
  n29 = LinearInterpolate(A=@n22.RGB, B=@n23.RGB, Alpha=Saturate(None=Power(Base=@n27, Exp=@n28) {constExponent=4}))
  n30 = TextureSampleParameter2D() {
   parameterName=PanningEmissiveMask
   samplerType=SAMPLERTYPE_LinearColor
   texture=/Game/Characters/Player/Female/Medium/Bodies/F_MED_ReptilianOcean_Battle/Textures/T_ReptilianOcean_Battle_Body_EMask
  }
  NamedRerouteDeclaration(None=StaticSwitchParameter(True=LinearInterpolate(A=@n21, B=@n29, Alpha=@n30.R), False=NamedRerouteUsage()) {parameterName=EnablePanningEmissive}) {name=PanningEmissiveColor, nodeColor=<rgba 1 0 0.5859 1>}
  NamedRerouteDeclaration(None=@n6.Tech Art Mask) {name=FXMask, nodeColor=<rgba 0.6406 0 1 1>}
  n31 = NamedRerouteUsage()
  n32 = Multiply(A=@n31, B=ScalarParameter() {defaultValue=1, parameterName=SnakesEyesEmissiveIntensity})
  n33 = ChannelMaskParameter(None=NamedRerouteUsage()) {maskChannel=Blue, parameterName=GlowEyesEmissiveMask, defaultValue=<rgba 0 0 1 0>}
  NamedRerouteDeclaration(None=StaticSwitchParameter(True=LinearInterpolate(A=@n31, B=@n32, Alpha=@n33.R), False=NamedRerouteUsage()) {parameterName=EnableSnakesEyesEffect}) {name=OutEmissive, nodeColor=<rgba 0 1 0.9063 1>}
  n34 = Power(Base=NamedRerouteUsage()) {constExponent=4}
  n35 = Multiply(
   A=Multiply(A=@n19.RGB, B=Saturate(None=@n34))
   B=SmoothStep(Min=ScalarParameter() {defaultValue=0.25, parameterName=ExtraThinFilmScalesBlend}, Value=NamedRerouteUsage()) {constMin=0.25}
  )
  n36 = Clamp(None=OneMinus(None=NamedRerouteUsage())) {minDefault=0.3}
  NamedRerouteDeclaration(None=StaticSwitchParameter(True=Multiply(A=@n35, B=@n36), False=@n35) {parameterName=EnablePatternsEffect}) {name=ExtraThinFilmSwipe, nodeColor=<rgba 1 0 0.8906 1>}
  n37 = NamedRerouteUsage()
  n38 = MaterialFunctionCall(Normal=NamedRerouteUsage(), Flatness=ScalarParameter() {defaultValue=0.8, parameterName=ExtraThinFilmNormalContribution}) {materialFunction=&FlattenNormal}
  n39 = Add(A=Fresnel(Normal=Transform(None=@n38.Result)) {baseReflectFraction=0, exponent=0.1}, B=Multiply(A=Time()) {constB=-0.05}) {constB=0.5}
  n40 = ConstantBiasScale(None=Sine(None=Multiply(A=@n39) {constB=3.5}))
  n41 = Multiply(
   A=NamedRerouteDeclaration(None=TextureSample(UVs=@n40) {texture=&T_ThinFilm_Spectrum_COLOR}.RGB) {name=ScaleThinFilm_1, nodeColor=<rgba 0.1016 0 1 1>}
   B=ScalarParameter() {defaultValue=0.075, parameterName=ExtraThinFilmIntensity}
  ) {constB=0.075}
  n42 = NamedRerouteUsage()
  n43 = NamedRerouteDeclaration(None=Multiply(A=Multiply(A=@n41, B=@n42), B=NamedRerouteUsage())) {name=ExtraThinFilmSwipe_1, nodeColor=<rgba 1 0 0.8906 1>}
  n44 = TextureCoordinate()
  n45 = ScalarParameter() {defaultValue=1, parameterName=PatternTiling}
  n46 = VectorParameter() {parameterName=PatternOffset, defaultValue=<rgba 0 0 0 1>}
  n47 = TextureSample(UVs=Add(A=Multiply(A=@n44, B=@n45) {constB=2}, B=ComponentMask(None=@n46.RGB) {g=true, r=true})) {
   samplerType=SAMPLERTYPE_LinearGrayscale
   texture=/Game/Characters/Player/Female/Medium/Bodies/F_MED_ReptilianOcean_Sleek/Textures/T_ReptilianOcean_Sleek_Pattern_SDF
  }
  n48 = Add(
   A=Multiply(A=@n47.R, B=ScalarParameter() {defaultValue=2.1, parameterName=TilingLayers}) {constB=2.5}
   B=Multiply(A=Time(), B=ScalarParameter() {parameterName=TilingSpeed}) {constB=0}
  )
  n49 = OneMinus(None=Step(Y=Sine(None=@n48), X=ScalarParameter() {defaultValue=0.75, parameterName=PatternThickness}) {constY=0.1})
  n50 = NamedRerouteUsage()
  n51 = StaticSwitchParameter(
   True=LinearInterpolate(
    A=Add(A=@n37, B=@n43)
    B=VectorParameter() {parameterName=PatternColor, defaultValue=<rgba 0.0123 0.0115 0.026 0.85>}.RGB
    Alpha=Multiply(A=@n49, B=@n50)
   ) {constB=0}
   False=NamedRerouteUsage()
  ) {parameterName=EnableScalesEffect}
  n52 = VectorParameter() {parameterName=EyeMarkColor, defaultValue=<rgba 0.0104 0.0047 0.0036 1>}
  n53 = Multiply(A=ComponentMask(None=NamedRerouteUsage()) {r=true}, B=ScalarParameter() {parameterName=EyeEffect, sliderMax=1})
  NamedRerouteDeclaration(None=StaticSwitchParameter(True=LinearInterpolate(A=@n51, B=@n52.RGB, Alpha=@n53) {constB=0}, False=@n51) {parameterName=EnableEyeEffect}) {name=BaseColor_1, nodeColor=<rgba 1 0 0.1875 1>}
  n54 = NamedRerouteUsage()
  n55 = TextureSample(UVs=Multiply(A=TextureCoordinate(), B=Constant() {r=1.5}) {constB=2}) {
   samplerType=SAMPLERTYPE_LinearColor
   texture=/Game/Characters/Player/Female/Medium/Bodies/F_MED_ReptilianOcean_Sleek/Textures/T_ReptilianOcean_Sleek_Pattern_Noise
  }
  n56 = Add(A=Subtract(A=Multiply(A=@n55.R) {constB=2}), B=Multiply(A=Time(), B=Constant() {r=-0.1}) {constB=0.2})
  n57 = Saturate(None=ConstantBiasScale(None=Sine(None=@n56)))
  n58 = CurveAtlasRowParameter(CurveTime=@n57) {
   atlas=/Game/Characters/Player/Female/Medium/Bodies/F_MED_ReptilianOcean_Sleek/Textures/CA_ReptilianOcean
   curve=/Game/Characters/Player/Female/Medium/Bodies/F_MED_ReptilianOcean_Sleek/Textures/C_ReptilianOceanSleekCoarse
  }
  n59 = Multiply(A=Multiply(A=@n57, B=ScalarParameter() {defaultValue=0.6, parameterName=SpotsOpacityBlend}) {constB=0.5}, B=NamedRerouteUsage())
  n60 = Add(
   A=StaticSwitchParameter(True=LinearInterpolate(A=@n54, B=@n58, Alpha=@n59), False=NamedRerouteUsage()) {parameterName=EnableScalesSpotsColors}
   B=Multiply(A=NamedRerouteUsage(), B=ScalarParameter() {defaultValue=0.1, parameterName=ExtraThinFilmDiffuseIntensity}) {constB=0.075}
  )
  n61 = VectorParameter() {parameterName=PatternColor, defaultValue=<rgba 0.0123 0.0115 0.026 0.85>}
  n62 = ScalarParameter() {defaultValue=2.1, parameterName=PatternLayers}
  n63 = ComponentMask(None=NamedRerouteUsage()) {g=true, r=true}
  n64 = ScalarParameter() {defaultValue=-0.005, parameterName=PatternsNormalContribution}
  n65 = TextureSample(UVs=Add(A=TextureCoordinate(), B=Multiply(A=@n63, B=@n64))) {
   samplerType=SAMPLERTYPE_LinearGrayscale
   texture=&T_ReptilianOcean_Sleek_Pattern_SDF
  }
  n66 = VertexInterpolator(VS=PreSkinnedPosition())
  n67 = Add(A=Multiply(A=Time()) {constB=0.15}, B=Divide(A=ComponentMask(None=@n66.PS) {b=true}) {constB=-50})
  n68 = NamedRerouteUsage()
  n69 = LinearInterpolate(A=NamedRerouteDeclaration(None=Sine(None=@n67)) {name=SecondarySineWave, nodeColor=<rgba 0.2656 0 1 1>}, B=@n68, Alpha=Saturate(None=@n68))
  n70 = ScalarParameter() {defaultValue=0.125, parameterName=PatternInflateAmount}
  n71 = Add(
   A=Add(A=Multiply(A=@n62, B=@n65.G) {constB=2.5}, B=Multiply(A=@n69, B=@n70) {constB=0.1})
   B=Multiply(A=Time(), B=ScalarParameter() {parameterName=PatternTilingSpeed}) {constB=0}
  )
  n72 = OneMinus(None=Step(Y=Sine(None=@n71), X=ScalarParameter() {defaultValue=0.75, parameterName=PatternThickness}) {constY=0.1})
  n73 = ScalarParameter() {defaultValue=0.6, parameterName=PatternsScalesBlend}
  n74 = NamedRerouteUsage()
  n75 = Multiply(
   A=@n61.A
   B=Multiply(A=NamedRerouteDeclaration(None=@n72) {name=PatternsMask, nodeColor=<rgba 0.2187 1 0 1>}, B=SmoothStep(Max=@n73, Value=@n74) {constMax=0.3})
  )
  n76 = StaticSwitchParameter(
   True=StaticSwitchParameter(True=LinearInterpolate(A=@n60, B=@n61.RGB, Alpha=@n75) {constB=0}, False=@n60) {parameterName=EnablePatternsEffect}
   False=NamedRerouteUsage()
  ) {parameterName=EnableScalesEffect}
  n77 = VectorParameter() {parameterName=EyeMarkColor, defaultValue=<rgba 0.0104 0.0047 0.0036 1>}
  n78 = Multiply(A=ComponentMask(None=NamedRerouteUsage()) {r=true}, B=NamedRerouteUsage())
  NamedRerouteDeclaration(None=StaticSwitchParameter(True=LinearInterpolate(A=@n76, B=@n77.RGB, Alpha=@n78) {constB=0}, False=@n76) {parameterName=EnableEyeEffect}) {name=BaseColor, nodeColor=<rgba 1 0 0.1875 1>}
  n79 = NamedRerouteUsage()
  n80 = Multiply(A=NamedRerouteUsage(), B=ScalarParameter() {defaultValue=0.05, parameterName=ExtraThinFilmEmissiveIntensity})
  NamedRerouteDeclaration(None=StaticSwitchParameter(True=Add(A=@n79, B=@n80), False=@n79) {parameterName=EnableScalesEffect}) {name=ScalesEmissive, nodeColor=<rgba 1 0.2578 0 1>}
  NamedRerouteDeclaration(None=@n5) {name=EyesSphereMasks, nodeColor=<rgba 0 1 0.8828 1>}
  n81 = TextureCoordinate()
  n82 = ComponentMask(None=VectorParameter() {parameterName=ScalesExtraSweepScale, defaultValue=<rgba 1 1 0 1>}.RGB) {g=true, r=true}
  n83 = TextureSample(UVs=Panner(Coordinate=Multiply(A=@n81, B=@n82) {constB=4}) {speedY=-0.015}) {
   samplerType=SAMPLERTYPE_LinearColor
   texture=&T_ReptilianOcean_Sleek_Pattern_Noise
  }
  n84 = Subtract(A=Saturate(None=SmoothStep(Value=@n83.G) {constMin=0.1}), B=LinearInterpolate(Alpha=NamedRerouteUsage()) {constB=0.5})
  n85 = Multiply(
   A=LinearInterpolate(Alpha=Saturate(None=@n84)) {constA=-0.2}
   B=ScalarParameter() {defaultValue=0.07, parameterName=ScalesSecondaryNormalInfluence}
  ) {constB=0.5}
  n86 = NamedRerouteUsage()
  n87 = Multiply(
   A=TextureCoordinate()
   B=ComponentMask(None=VectorParameter() {parameterName=ScaleNoiseTiling, defaultValue=<rgba 0.75 1 0 1>}.RGB) {g=true, r=true}
  ) {constB=2}
  n88 = VectorParameter() {parameterName=ScaleNoiseSpeed, defaultValue=<rgba 0 0.075 0 1>}
  n89 = TextureSample(UVs=Panner(Coordinate=Multiply(A=@n87) {constB=2.5}, Speed=ComponentMask(None=@n88.RGB) {g=true, r=true}) {speedY=0.075}) {
   samplerType=SAMPLERTYPE_LinearColor
   texture=&T_ReptilianOcean_Sleek_Pattern_Noise
  }
  n90 = LinearInterpolate(Alpha=NamedRerouteDeclaration(None=@n89.R) {name=PanningUpNoise, nodeColor=<rgba 0.8516 1 0 1>})
  n91 = Multiply(
   A=StaticSwitchParameter(True=Multiply(A=@n86, B=@n90), False=@n86) {parameterName=EnableScaleNoise}
   B=ScalarParameter() {defaultValue=0.12, parameterName=ScalesSweepNormalRotateAmount}
  ) {constB=0.15}
  n92 = NamedRerouteUsage()
  NamedRerouteDeclaration(None=Multiply(A=Add(A=@n85, B=@n91), B=SmoothStep(Value=@n92) {constMin=0.1})) {name=NormalsRotationAmount, nodeColor=<rgba 0 1 0.5547 1>}
  n93 = VertexInterpolator(VS=PreSkinnedPosition())
  NamedRerouteDeclaration(None=MaterialFunctionCall!(A=Divide(A=@n93.PS) {constB=100}, B=Constant3Vector() {constant=<rgba 0 0 0.9 0>}, Bounds=Constant() {r=0.35}, Edge Falloff=Constant()) {materialFunction=/Engine/Functions/Engine_MaterialFunctions01/Gradient/BoxMask-3D}.Result) {name=Battle Waist Mask, nodeColor=<rgba 1 0.5391 0 1>}
  NamedRerouteDeclaration(None=@n6.Result) {name=inMaterialAttributes, nodeColor=<rgba 1 0.0234 0 1>}
  n94 = SetMaterialAttributes!(
   MaterialAttributes=@n20.MaterialAttributes
   Base Color=NamedRerouteUsage()
   Normal=NamedRerouteUsage()
   World Position Offset=NamedRerouteUsage()
   Emissive Color=NamedRerouteUsage()
  ) {attributeSetTypes [
    #1
    #2
    #4
    #3
   ]}
  NamedRerouteDeclaration(None=@n94) {name=outMaterialAttributes, nodeColor=<rgba 0.5938 0 1 1>}
  n95 = PreSkinnedPosition()
  n96 = ComponentMask(None=ShaderStageSwitch(PixelShader=VertexInterpolator(VS=@n95).PS, VertexShader=@n95)) {b=true}
  n97 = ScalarParameter() {defaultValue=300, parameterName=ScalesSweepSize}
  n98 = Time()
  n99 = ScalarParameter() {defaultValue=0.09, parameterName=ScalesSweepSpeed}
  n100 = Add(A=Divide(A=@n96, B=Multiply(A=@n97) {constB=0.75}), B=Multiply(A=Add(A=@n98) {constB=5}, B=Multiply(A=@n99) {constB=0.75}) {constB=0.15})
  n101 = Saturate(None=Abs(None=Sine(None=@n100)))
  n102 = ScalarParameter() {defaultValue=0.5, parameterName=ScalesSecondSweepRandomBlend}
  n103 = NamedRerouteUsage()
  n104 = Saturate(None=Subtract(A=@n101, B=LinearInterpolate(B=@n102, Alpha=@n103) {constB=0.5}))
  n105 = Saturate(None=Add(A=Multiply(A=@n104, B=@n101), B=Multiply(A=@n101) {constB=0.5}))
  n106 = ScalarParameter() {defaultValue=1, parameterName=ScalesSweepRandomBlend}
  n107 = NamedRerouteUsage()
  n108 = Time()
  n109 = ScalarParameter() {defaultValue=0.09, parameterName=ScalesSweepSpeed}
  n110 = Sine(None=Add(A=Multiply(A=@n108, B=@n109) {constB=0.15}, B=Divide(A=@n96, B=@n97) {constB=-125}))
  n111 = StaticSwitch(True=Abs(None=@n110), False=@n110, Value=StaticBoolParameter() {parameterName=ScalesEnableDoubleSweep})
  n112 = NamedRerouteDeclaration(None=Power(Base=Saturate(None=@n111), Exp=ScalarParameter() {defaultValue=2, parameterName=ScalesSweeoSmoothness}) {constExponent=8}) {name=GradientMask, nodeColor=<rgba 0 1 0.3906 1>}
  n113 = Saturate(None=Subtract(A=@n112, B=LinearInterpolate(B=@n106, Alpha=@n107)))
  n114 = Saturate(None=Add(A=Multiply(A=@n113, B=@n112), B=Multiply(A=@n112) {constB=0.75}))
  NamedRerouteDeclaration(None=LinearInterpolate(A=Multiply(A=@n105) {constB=0.9}, B=@n114, Alpha=@n114)) {name=ScaleSweepBump, nodeColor=<rgba 0 0.7422 1 1>}
  n115 = Clamp(None=NamedRerouteUsage()) {minDefault=-0.15}
  n116 = OneMinus(None=ComponentMask(None=NamedRerouteUsage()) {b=true})
  n117 = Multiply(
   A=Multiply(A=Multiply(A=@n115, B=@n116), B=NamedRerouteUsage())
   B=ScalarParameter() {defaultValue=0.9, parameterName=ScalesSwipeWPOIntensity}
  ) {constB=0.8}
  NamedRerouteDeclaration(None=StaticSwitchParameter(True=Transform(None=@n117), False=Constant()) {parameterName=EnableScalesEffect}) {name=OutWPO, nodeColor=<rgba 0 0.8359 1 1>}
  NamedRerouteDeclaration(None=@n20.Normal) {name=InNormal, nodeColor=<rgba 1 0 0.3984 1>}
  n118 = NamedRerouteUsage()
  n119 = Transform(None=Constant3Vector() {constant=<rgba 0 0 -1 0>}) {transformSourceType=TRANSFORMSOURCE_World, transformType=TRANSFORM_Tangent}
  n120 = VertexNormalWS()
  n121 = RotateAboutAxis(
   NormalizedRotationAxis=CrossProduct(A=Normalize(VectorInput=@n119), B=Transform(None=@n120) {transformSourceType=TRANSFORMSOURCE_World, transformType=TRANSFORM_Tangent})
   RotationAngle=NamedRerouteUsage()
   PivotPoint=Constant()
   Position=@n118
  )
  NamedRerouteDeclaration(None=StaticSwitchParameter(True=Add(A=@n118, B=@n121), False=@n118) {parameterName=EnableScalesEffect}) {name=OutNormal, nodeColor=<rgba 1 0 0.1406 1>}
  NamedRerouteDeclaration(None=@n6.M) {name=inMask, nodeColor=<rgba 0.7578 1 0 1>}
  Power(Base=@n101)
  SmoothStep(Value=@n47.R)
  Subtract(A=Multiply(A=@n47.R) {constB=2})
  TextureSample() {
   samplerType=SAMPLERTYPE_Masks
   texture=/Game/Characters/Player/Female/Medium/Bodies/F_MED_ReptilianOcean_Battle/Textures/T_ReptilianOcean_Battle_Body_FX
  }
  TextureSample() {
   samplerType=SAMPLERTYPE_Masks
   texture=/Game/Characters/Player/Female/Medium/Bodies/F_MED_ReptilianOcean_Battle/Textures/T_ReptilianOcean_Battle_Body_M
  }
 ]
}
