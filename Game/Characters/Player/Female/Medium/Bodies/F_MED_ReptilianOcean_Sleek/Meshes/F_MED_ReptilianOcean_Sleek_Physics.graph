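# Physics asset for the female medium ReptilianOcean Sleek body, providing capsule-based dynamics for skirt-like cloth volume bones around the hips.
# Front, back and side bone chains hang from the pelvis on free-swinging joints, with mostly angle-limited joints below and a longer left B chain.
# Angularly free cross-constraints tie some neighboring side chains together.
# Pelvis, thigh and calf capsules supply body shapes, plus an oversized root capsule below the origin, likely a floor collider.
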PhysicsAsset /Game/Characters/Player/Female/Medium/Bodies/F_MED_ReptilianOcean_Sleek/Meshes/F_MED_ReptilianOcean_Sleek_Physics
bodies {
 calf_l [{
   shapeType=Capsule
   length=51.9392
   radius=7.0836
   shapeName=calf_l_capsule
   center=<-21.1277 -1.6458 -0.255>
   rotation=<rot -87.8858 -0.0001 0.0001>
  }]
 calf_r [{
   shapeType=Capsule
   length=50.1575
   radius=8.3329
   shapeName=calf_r_capsule
   center=<23.3183 1.1039 0.8608>
   rotation=<rot -87.8858 -0.0002 0.0001>
  }]
 dyn_Cloth_volume_bk02 [{shapeType=Capsule, length=15, radius=9.0301, shapeName=dyn_Cloth_volume_bk02_capsule, center=<-6.9103 0 0>}]
 dyn_Cloth_volume_bk03 [{shapeType=Capsule, length=19.3641, radius=9.7088, shapeName=dyn_Cloth_volume_bk03_capsule, center=<-7.9388 -4.3218 0>}]
 dyn_Cloth_volume_bk04 [{shapeType=Capsule, length=18.3641, radius=10.7088, shapeName=dyn_Cloth_volume_bk04_capsule, center=<-7.9388 -9.3609 0>}]
 dyn_Cloth_volume_fn02 [{shapeType=Capsule, length=15, radius=9.0301, shapeName=dyn_Cloth_volume_fn02_capsule, center=<-6.9103 0 0>}]
 dyn_Cloth_volume_fn03 [{shapeType=Capsule, length=22.0737, radius=7.5, shapeName=dyn_Cloth_volume_fn03_capsule, center=<-8.3835 -6.2409 0>}]
 dyn_Cloth_volume_fn04 [{shapeType=Capsule, length=20.5737, radius=9.5, shapeName=dyn_Cloth_volume_fn04_capsule, center=<-5.808 -6.2409 0>}]
 dyn_Cloth_volume_l_A02 [{shapeType=Capsule, length=15, radius=9.0301, shapeName=dyn_Cloth_volume_l_A02_capsule, center=<-6.9103 0 0>, rotation=<rot 0 0.0001 7.4926>}]
 dyn_Cloth_volume_l_A03 [{
   shapeType=Capsule
   length=9.965
   radius=9.7088
   shapeName=dyn_Cloth_volume_l_A03_capsule
   center=<-7.9388 -4.3218 0>
   rotation=<rot 0 0.0001 7.4926>
  }]
 dyn_Cloth_volume_l_A04 [{
   shapeType=Capsule
   length=9.965
   radius=9.7088
   shapeName=dyn_Cloth_volume_l_A04_capsule
   center=<-7.9388 -6.1481 0>
   rotation=<rot 0 0.0001 7.4926>
  }]
 dyn_Cloth_volume_l_B01 [{
   shapeType=Capsule
   length=5.2754
   radius=4.6632
   shapeName=dyn_Cloth_volume_l_B01_capsule
   center=<-2.6326 -1.7174 1.1123>
   rotation=<rot 7.5066 -90 21.9074>
  }]
 dyn_Cloth_volume_l_B02 [{
   shapeType=Capsule
   length=16.6441
   radius=9.7743
   shapeName=dyn_Cloth_volume_l_B02_capsule
   center=<-10.1027 -8.9464 0.069>
   rotation=<rot 0.0001 0 -6.5015>
  }]
 dyn_Cloth_volume_l_B03 [{
   shapeType=Capsule
   length=14.6564
   radius=10.4317
   shapeName=dyn_Cloth_volume_l_B03_capsule
   center=<-12.371 -6.9812 -0.4556>
   rotation=<rot 0.0003 0 14.8641>
  }]
 dyn_Cloth_volume_l_B04 [{
   shapeType=Capsule
   length=14.0012
   radius=12.4409
   shapeName=dyn_Cloth_volume_l_B04_capsule
   center=<-14.1197 -6.2486 2.9357>
   rotation=<rot 0.0002 -5.5211 20.5559>
  }]
 dyn_Cloth_volume_l_B05 [{
   shapeType=Capsule
   length=11.5192
   radius=10.8931
   shapeName=dyn_Cloth_volume_l_B05_capsule
   center=<-10.1972 1.3704 4.0717>
   rotation=<rot 0.0003 -29.6541 20.1927>
  }]
 dyn_Cloth_volume_l_C02 [{shapeType=Capsule, length=15, radius=9.0301, shapeName=dyn_Cloth_volume_l_C02_capsule, center=<-6.9103 0 0>}]
 dyn_Cloth_volume_l_C03 [{shapeType=Capsule, length=7.5918, radius=12.082, shapeName=dyn_Cloth_volume_l_C03_capsule, center=<-12.0332 -1.8625 0>}]
 dyn_Cloth_volume_l_C04 [{
   shapeType=Capsule
   length=19.3968
   radius=13.0776
   shapeName=dyn_Cloth_volume_l_B04_capsule
   center=<-15.5893 -3.2398 1.4651>
   rotation=<rot -6.3762 -5.5213 -10.0781>
  }]
 dyn_Cloth_volume_r_A02 [{shapeType=Capsule, length=15, radius=9.0301, shapeName=dyn_Cloth_volume_r_A02_capsule, center=<-6.9103 0 0>}]
 dyn_Cloth_volume_r_A03 [{shapeType=Capsule, length=19.465, radius=9.2088, shapeName=dyn_Cloth_volume_r_A03_capsule, center=<-3.3382 -8.4156 0>}]
 dyn_Cloth_volume_r_A04 [{shapeType=Capsule, length=19.465, radius=9.2088, shapeName=dyn_Cloth_volume_r_A04_capsule, center=<-5.1905 -7.5609 0>}]
 dyn_Cloth_volume_r_B02 [{shapeType=Capsule, length=15, radius=9.0301, shapeName=dyn_Cloth_volume_r_B02_capsule, center=<-6.9103 0 0>}]
 dyn_Cloth_volume_r_B03 [{shapeType=Capsule, length=13.4749, radius=10.7088, shapeName=dyn_Cloth_volume_r_B03_capsule, center=<-7.9388 -8.0886 0>}]
 dyn_Cloth_volume_r_B04 [{
   shapeType=Capsule
   length=14.739
   radius=11.2446
   shapeName=dyn_Cloth_volume_r_B04_capsule
   center=<-12.6643 -8.8355 0>
   rotation=<rot -0.0001 -6.8696 0>
  }]
 dyn_Cloth_volume_r_C02 [{shapeType=Capsule, length=15, radius=9.0301, shapeName=dyn_Cloth_volume_r_C02_capsule, center=<-6.9103 0 0>}]
 dyn_Cloth_volume_r_C03 [{shapeType=Capsule, length=9.965, radius=9.7088, shapeName=dyn_Cloth_volume_r_C03_capsule, center=<-7.9388 -4.3218 0>}]
 dyn_Cloth_volume_r_C04 [{shapeType=Capsule, length=9.965, radius=9.7088, shapeName=dyn_Cloth_volume_r_C04_capsule, center=<-7.9388 -8.2326 0>}]
 pelvis [{shapeType=Capsule, length=5.5401, radius=11.5835, shapeName=pelvis_capsule}]
 root [{
   shapeType=Capsule
   length=43.7212
   radius=184.4596
   shapeName=root_capsule
   center=<1.513 -3.0618 -182.3861>
   rotation=<rot 89.8331 -179.9464 -179.9464>
  }]
 thigh_l [{
   shapeType=Capsule
   length=42.8729
   radius=7.5247
   shapeName=thigh_l_capsule
   center=<-17.3631 0.67 -0.5841>
   rotation=<rot -88.0732 -0.0002 0.0002>
  }]
 thigh_r [{
   shapeType=Capsule
   length=42.4024
   radius=7.9952
   shapeName=thigh_r_capsule
   center=<19.1593 0 -0.5557>
   rotation=<rot -88.0732 -0.0002 0.0002>
  }]
}
constraints [
 {bone1Name=dyn_Cloth_volume_bk04, bone2Name=dyn_Cloth_volume_bk03, swing1LimitDegrees=45, swing2LimitDegrees=45, twistLimitDegrees=45}
 {bone1Name=dyn_Cloth_volume_fn04, bone2Name=dyn_Cloth_volume_fn03, swing1LimitDegrees=25, swing2LimitDegrees=25, twistLimitDegrees=15}
 {bone1Name=dyn_Cloth_volume_l_A04, bone2Name=dyn_Cloth_volume_l_A03, swing1LimitDegrees=45, swing2LimitDegrees=45, twistLimitDegrees=45}
 {bone1Name=dyn_Cloth_volume_l_B04, bone2Name=dyn_Cloth_volume_l_B03, swing1LimitDegrees=45, swing2LimitDegrees=45, twistLimitDegrees=45}
 {bone1Name=dyn_Cloth_volume_l_B05, bone2Name=dyn_Cloth_volume_l_B04, swing1LimitDegrees=45, swing2LimitDegrees=45, twistLimitDegrees=45}
 {bone1Name=dyn_Cloth_volume_l_C04, bone2Name=dyn_Cloth_volume_l_C03, swing1LimitDegrees=45, swing2LimitDegrees=45, twistLimitDegrees=45}
 {bone1Name=dyn_Cloth_volume_r_A04, bone2Name=dyn_Cloth_volume_r_A03, swing1LimitDegrees=45, swing2LimitDegrees=45, twistLimitDegrees=45}
 {bone1Name=dyn_Cloth_volume_r_B04, bone2Name=dyn_Cloth_volume_r_B03, swing1LimitDegrees=45, swing2LimitDegrees=45, twistLimitDegrees=45}
 {bone1Name=dyn_Cloth_volume_r_C04, bone2Name=dyn_Cloth_volume_r_C03, swing1LimitDegrees=45, swing2LimitDegrees=45, twistLimitDegrees=45}
 {bone1Name=dyn_Cloth_volume_fn03, bone2Name=dyn_Cloth_volume_fn02, swing1LimitDegrees=25, swing2LimitDegrees=25, twistLimitDegrees=15}
 {
  bone1Name=dyn_Cloth_volume_fn02
  bone2Name=pelvis
  swing1LimitDegrees=45
  swing1Motion=Free
  swing2LimitDegrees=45
  swing2Motion=Free
  twistLimitDegrees=45
  twistMotion=Free
 }
 {bone1Name=dyn_Cloth_volume_bk03, bone2Name=dyn_Cloth_volume_bk02, swing1LimitDegrees=45, swing2LimitDegrees=45, twistLimitDegrees=45}
 {
  bone1Name=dyn_Cloth_volume_bk02
  bone2Name=pelvis
  swing1LimitDegrees=45
  swing1Motion=Free
  swing2LimitDegrees=45
  swing2Motion=Free
  twistLimitDegrees=45
  twistMotion=Free
 }
 {
  bone1Name=dyn_Cloth_volume_l_A03
  bone2Name=dyn_Cloth_volume_l_A02
  swing1LimitDegrees=45
  swing1Motion=Free
  swing2LimitDegrees=45
  swing2Motion=Free
  twistLimitDegrees=45
  twistMotion=Free
 }
 {
  bone1Name=dyn_Cloth_volume_l_A02
  bone2Name=pelvis
  swing1LimitDegrees=45
  swing1Motion=Free
  swing2LimitDegrees=45
  swing2Motion=Free
  twistLimitDegrees=45
  twistMotion=Free
 }
 {bone1Name=dyn_Cloth_volume_l_B03, bone2Name=dyn_Cloth_volume_l_B02, swing1LimitDegrees=45, swing2LimitDegrees=45, twistLimitDegrees=45}
 {bone1Name=dyn_Cloth_volume_l_B02, bone2Name=dyn_Cloth_volume_l_B01, swing1LimitDegrees=5, swing2LimitDegrees=5, twistLimitDegrees=5}
 {bone1Name=dyn_Cloth_volume_l_C03, bone2Name=dyn_Cloth_volume_l_C02, swing1LimitDegrees=45, swing2LimitDegrees=45, twistLimitDegrees=45}
 {
  bone1Name=dyn_Cloth_volume_l_C02
  bone2Name=pelvis
  swing1LimitDegrees=45
  swing1Motion=Free
  swing2LimitDegrees=45
  swing2Motion=Free
  twistLimitDegrees=45
  twistMotion=Free
 }
 {bone1Name=dyn_Cloth_volume_r_A03, bone2Name=dyn_Cloth_volume_r_A02, swing1LimitDegrees=45, swing2LimitDegrees=45, twistLimitDegrees=45}
 {
  bone1Name=dyn_Cloth_volume_r_A02
  bone2Name=pelvis
  swing1LimitDegrees=45
  swing1Motion=Free
  swing2LimitDegrees=45
  swing2Motion=Free
  twistLimitDegrees=45
  twistMotion=Free
 }
 {bone1Name=dyn_Cloth_volume_r_B03, bone2Name=dyn_Cloth_volume_r_B02, swing1LimitDegrees=45, swing2LimitDegrees=45, twistLimitDegrees=45}
 {
  bone1Name=dyn_Cloth_volume_r_B02
  bone2Name=pelvis
  swing1LimitDegrees=45
  swing1Motion=Free
  swing2LimitDegrees=45
  swing2Motion=Free
  twistLimitDegrees=45
  twistMotion=Free
 }
 {bone1Name=dyn_Cloth_volume_r_C03, bone2Name=dyn_Cloth_volume_r_C02, swing1LimitDegrees=45, swing2LimitDegrees=45, twistLimitDegrees=45}
 {
  bone1Name=dyn_Cloth_volume_r_C02
  bone2Name=pelvis
  swing1LimitDegrees=45
  swing1Motion=Free
  swing2LimitDegrees=45
  swing2Motion=Free
  twistLimitDegrees=45
  twistMotion=Free
 }
 {
  bone1Name=dyn_Cloth_volume_l_B01
  bone2Name=pelvis
  swing1LimitDegrees=45
  swing1Motion=Free
  swing2LimitDegrees=45
  swing2Motion=Free
  twistLimitDegrees=45
  twistMotion=Free
 }
 {
  bone1Name=dyn_Cloth_volume_r_B03
  bone2Name=dyn_Cloth_volume_r_C03
  swing1LimitDegrees=45
  swing1Motion=Free
  swing2LimitDegrees=45
  swing2Motion=Free
  twistLimitDegrees=45
  twistMotion=Free
 }
 {
  bone1Name=dyn_Cloth_volume_r_A03
  bone2Name=dyn_Cloth_volume_r_B03
  swing1LimitDegrees=45
  swing1Motion=Free
  swing2LimitDegrees=45
  swing2Motion=Free
  twistLimitDegrees=45
  twistMotion=Free
 }
 {
  bone1Name=dyn_Cloth_volume_r_B04
  bone2Name=dyn_Cloth_volume_r_C04
  swing1LimitDegrees=45
  swing1Motion=Free
  swing2LimitDegrees=45
  swing2Motion=Free
  twistLimitDegrees=45
  twistMotion=Free
 }
 {
  bone1Name=dyn_Cloth_volume_l_C04
  bone2Name=dyn_Cloth_volume_l_B05
  swing1LimitDegrees=45
  swing1Motion=Free
  swing2LimitDegrees=45
  swing2Motion=Free
  twistLimitDegrees=45
  twistMotion=Free
 }
 {
  bone1Name=dyn_Cloth_volume_l_C04
  bone2Name=dyn_Cloth_volume_l_B04
  swing1LimitDegrees=45
  swing1Motion=Free
  swing2LimitDegrees=45
  swing2Motion=Free
  twistLimitDegrees=45
  twistMotion=Free
 }
 {
  bone1Name=dyn_Cloth_volume_l_B03
  bone2Name=dyn_Cloth_volume_l_C03
  swing1LimitDegrees=45
  swing1Motion=Free
  swing2LimitDegrees=45
  swing2Motion=Free
  twistLimitDegrees=45
  twistMotion=Free
 }
]
